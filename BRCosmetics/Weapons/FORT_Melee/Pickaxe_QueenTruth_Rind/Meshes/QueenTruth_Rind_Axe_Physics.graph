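# Physics asset for the QueenTruth Rind pickaxe cosmetic, defining collision and joint limits for a figure on a clock root with head, hair, arms and feet.
# Each bone has one capsule body, with a large one on Root_Clock_jnt anchoring the head, shoulders and feet.
# Two-link hair chains hang from the head and elbows follow shoulders, all with narrow swing and twist limits.
# The left elbow also constrains directly to the root, unlike the right.
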PhysicsAsset /BRCosmetics/Weapons/FORT_Melee/Pickaxe_QueenTruth_Rind/Meshes/QueenTruth_Rind_Axe_Physics
bodies {
 Elbow_l_jnt [{
   shapeType=Capsule
   length=0.01
   radius=1.5
   shapeName=Elbow_l_jnt_capsule
   center=<-0.0255 0.0101 -0.2883>
   rotation=<rot 44.5064 92.0284 -89.9992>
  }]
 Elbow_r_jnt [{
   shapeType=Capsule
   length=0.01
   radius=1.5
   shapeName=Elbow_r_jnt_capsule
   center=<0.0255 -0.0101 0.2883>
   rotation=<rot 44.5064 92.0284 90.0008>
  }]
 Foot_l_jnt [{
   shapeType=Capsule
   length=3.518
   radius=1.4879
   shapeName=Foot_l_jnt_capsule
   center=<3.5328 0.2239 -0.2111>
   rotation=<rot 2.0253 -80.5198 -82.4544>
  }]
 Foot_r_jnt [{
   shapeType=Capsule
   length=3.518
   radius=1.4879
   shapeName=Foot_r_jnt_capsule
   center=<-3.4858 -0.2025 0.588>
   rotation=<rot 0 -74.3013 108.2188>
  }]
 Hair_01_l_jnt [{
   shapeType=Capsule
   length=5.3473
   radius=1.7937
   shapeName=Hair_01_l_jnt_capsule
   center=<3.6526 0.0137 -0.7661>
   rotation=<rot 40.6074 -90.0003 72.867>
  }]
 Hair_01_r_jnt [{
   shapeType=Capsule
   length=5.3473
   radius=1.7937
   shapeName=Hair_01_r_jnt_capsule
   center=<-3.6526 -0.0137 0.7661>
   rotation=<rot 40.6074 -90.0003 -107.133>
  }]
 Hair_02_l_jnt [{
   shapeType=Capsule
   length=4.3692
   radius=1.4988
   shapeName=Hair_02_l_jnt_capsule
   center=<3.5424 0.5064 -0.2262>
   rotation=<rot 0 99.4726 -88.7376>
  }]
 Hair_02_r_jnt [{
   shapeType=Capsule
   length=4.3692
   radius=1.4988
   shapeName=Hair_02_r_jnt_capsule
   center=<-3.5424 -0.5064 0.2262>
   rotation=<rot 0 99.4726 91.2624>
  }]
 Head_jnt [{
   shapeType=Capsule
   length=12.4632
   radius=4.9215
   shapeName=Head_jnt_capsule
   center=<9.0818 -0.0419 1.6943>
   rotation=<rot 0 88.3225 -58.9695>
  }]
 Root_Clock_jnt [{shapeType=Capsule, length=0, radius=12.4884, center=<-0.4816 -0.104 10.7337>, rotation=<rot 87.0766 -142.5651 -52.5291>}]
 Shoulder_l_jnt [{
   shapeType=Capsule
   length=0.0964
   radius=2.5407
   shapeName=Shoulder_l_jnt_capsule
   center=<3.9761 -0.128 0.1619>
   rotation=<rot 75.2648 125.0817 -55.8143>
  }]
 Shoulder_r_jnt [{
   shapeType=Capsule
   length=0.0964
   radius=2.5407
   shapeName=Shoulder_r_jnt_capsule
   center=<-3.9761 0.128 -0.1619>
   rotation=<rot 75.2648 125.0817 124.1857>
  }]
}
constraints [
 {bone1Name=Head_jnt, bone2Name=Root_Clock_jnt, swing1LimitDegrees=27, swing2LimitDegrees=0, twistLimitDegrees=30}
 {bone1Name=Hair_01_l_jnt, bone2Name=Head_jnt, swing1LimitDegrees=30, swing2LimitDegrees=10, twistLimitDegrees=10}
 {bone1Name=Hair_02_l_jnt, bone2Name=Hair_01_l_jnt, swing1LimitDegrees=25, swing2LimitDegrees=20, twistLimitDegrees=20}
 {bone1Name=Hair_02_r_jnt, bone2Name=Hair_01_r_jnt, swing1LimitDegrees=25, swing2LimitDegrees=20, twistLimitDegrees=20}
 {bone1Name=Hair_01_r_jnt, bone2Name=Head_jnt, swing1LimitDegrees=30, swing2LimitDegrees=5, twistLimitDegrees=5}
 {bone1Name=Elbow_l_jnt, bone2Name=Shoulder_l_jnt, swing1LimitDegrees=15, swing2LimitDegrees=15, twistLimitDegrees=5}
 {bone1Name=Foot_l_jnt, bone2Name=Root_Clock_jnt, swing1LimitDegrees=8, swing2LimitDegrees=2, twistLimitDegrees=2}
 {bone1Name=Shoulder_r_jnt, bone2Name=Root_Clock_jnt, swing1LimitDegrees=15, swing2LimitDegrees=2, twistLimitDegrees=2}
 {bone1Name=Elbow_r_jnt, bone2Name=Shoulder_r_jnt, swing1LimitDegrees=15, swing2LimitDegrees=5, twistLimitDegrees=5}
 {bone1Name=Foot_r_jnt, bone2Name=Root_Clock_jnt, swing1LimitDegrees=8, swing2LimitDegrees=2, twistLimitDegrees=2}
 {bone1Name=Elbow_l_jnt, bone2Name=Root_Clock_jnt, swing1LimitDegrees=15, swing2LimitDegrees=0, twistLimitDegrees=0}
 {bone1Name=Shoulder_l_jnt, bone2Name=Root_Clock_jnt, swing1LimitDegrees=15, swing2LimitDegrees=2, twistLimitDegrees=0}
]
