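# Animation Blueprint for the QueenTruth Rind pickaxe, animating its clock hands and adding secondary physics motion.
# The event graph converts scaled game time into hour and minute angles, which additively rotate two clock bones on the input pose.
# That cached pose feeds two Anim Dynamics chains for head, hair, shoulders and feet, switched by bIsInFrontEnd, with tighter limits in front end.
# bRigidBodyEnabled blends from that result to a RigidBody simulation.
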AnimBlueprint /BRCosmetics/Weapons/FORT_Melee/Pickaxe_QueenTruth_Rind/Meshes/QueenTruth_Rind_Axe_AnimBP
target_skeleton=&/QueenTruth_Rind_Axe_Skeleton
graphs [
 AnimGraph {nodes [
   n1 = RigidBody(ComponentPose=Local To Component(LocalPose=Use cached pose 'SavedInputPose'())) {
    bOverrideWorldGravity=true
    overridePhysicsAsset=&/QueenTruth_Rind_Axe_Physics
    simulationSpace=BaseBoneSpace
    baseBoneRef {boneName=b_AR_Root}
    componentAppliedLinearAccClamp=<3000 3000 3000>
    componentLinearAccScale=<0.4 0.3 0.3>
    componentLinearVelScale=<0.4 0.3 0.3>
    overrideWorldGravity {z=-600}
   }
   Output Pose(Result=Two Way Blend!(A=Use cached pose 'AnimDynamic'(), B=Component To Local(ComponentPose=@n1), bAlphaBoolEnabled:bool=Get bRigidBodyEnabled()) -> (Pose) {blendNode {alphaInputType=Bool, alphaBoolBlend {blendInTime=0.1, blendOption=Linear, blendOutTime=0.1}}}) {layerGroup=None}
   AnimGraphNode_SubInput_0 = Input Pose()
   n2 = "Transform (Modify) Bone - Bone: Clock_02_jnt"!(
    ComponentPose=Local To Component(LocalPose=@AnimGraphNode_SubInput_0)
    Alpha=1
    Translation="0,0,0"
    Rotation_Roll:float=Get HourAngle() -> (HourAngle:double)
    Rotation_Pitch:float=0
    Rotation_Yaw:float=0
    Scale="1,1,1"
   ) -> (Pose) {rotationMode=BMM_Additive, boneToModify {boneName=Clock_02_jnt}, rotation {roll=90}}
   n3 = Get MinuteAngle() -> (MinuteAngle:double)
   Save cached pose 'SavedInputPose'(Pose=Component To Local(ComponentPose="Transform (Modify) Bone - Bone: Clock_01_jnt"!(ComponentPose=@n2, Alpha=1, Translation="0,0,0", Rotation_Roll:float=@n3, Rotation_Pitch:float=0, Rotation_Yaw:float=0, Scale="1,1,1") -> (Pose) {rotationMode=BMM_Additive, boneToModify {boneName=Clock_01_jnt}})) {cacheName=SavedInputPose}
   n4 = Anim Dynamics - Bone: Head_jnt(ComponentPose=Local To Component(LocalPose=Use cached pose 'SavedInputPose'()), Alpha=1, ExternalForce="0,0,0") {
    angularDampingOverride=1
    angularSpringConstant=20000
    bAngularSpring=true
    bGravityOverrideInSimSpace=true
    bOverrideAngularDamping=true
    bUseGravityOverride=true
    linearSpringConstant=10000
    boundBone {boneName=Head_jnt}
    componentAppliedLinearAccClamp=<10000 10000 10000>
    componentLinearAccScale=<0.1 0.1 0.1>
    componentLinearVelScale=<0.1 0.1 0.1>
    gravityOverride {z=-980}
    physicsBodyDefinitions [{
      boundBone {boneName=Head_jnt}
      boxExtents=<12 12 12>
      constraintSetup {angularLimitsMax=<7 7 7>, angularLimitsMin=<-7 -7 -7>, angularTarget=<1 0 0>}
      localJointOffset=<-2 0 0>
     }]
    relativeSpaceBone {boneName=Root_Clock_jnt}
   }
   n5 = Anim Dynamics - Bone: Hair_01_l_jnt(ComponentPose=@n4, Alpha=1, ExternalForce="0,0,0") {
    angularDampingOverride=1
    angularSpringConstant=20000
    bAngularSpring=true
    bGravityOverrideInSimSpace=true
    bOverrideAngularDamping=true
    bUseGravityOverride=true
    linearSpringConstant=10000
    boundBone {boneName=Hair_01_l_jnt}
    componentAppliedLinearAccClamp=<10000 10000 10000>
    componentLinearAccScale=<0.1 0.1 0.1>
    componentLinearVelScale=<0.1 0.1 0.1>
    gravityOverride {z=-980}
    physicsBodyDefinitions [{
      boundBone {boneName=Hair_01_l_jnt}
      boxExtents=<10 10 10>
      constraintSetup {angularLimitsMax=<5 5 5>, angularLimitsMin=<-5 -5 -5>, angularTarget=<1 0 0>}
      localJointOffset=<-2 0 0>
     }]
    relativeSpaceBone {boneName=Root_Clock_jnt}
   }
   n6 = Anim Dynamics - Bone: Hair_01_r_jnt(ComponentPose=@n5, Alpha=1, ExternalForce="0,0,0") {
    angularDampingOverride=0.9999
    angularSpringConstant=20000
    bAngularSpring=true
    bGravityOverrideInSimSpace=true
    bOverrideAngularDamping=true
    bUseGravityOverride=true
    linearSpringConstant=10000
    boundBone {boneName=Hair_01_r_jnt}
    componentAppliedLinearAccClamp=<3000 3000 3000>
    componentLinearAccScale=<0.5 0.5 0.5>
    componentLinearVelScale=<0.5 0.5 0.5>
    gravityOverride {z=-980}
    physicsBodyDefinitions [{
      boundBone {boneName=Hair_01_r_jnt}
      boxExtents=<10 10 10>
      constraintSetup {angularLimitsMax=<5 5 5>, angularLimitsMin=<-5 -5 -5>, angularTarget=<1 0 0>}
      localJointOffset=<-2 0 0>
     }]
    relativeSpaceBone {boneName=Root_Clock_jnt}
   }
   n7 = Anim Dynamics - Bone: Shoulder_r_jnt(ComponentPose=@n6, Alpha=1, ExternalForce="0,0,0") {
    angularDampingOverride=0.99
    angularSpringConstant=20000
    bAngularSpring=true
    bGravityOverrideInSimSpace=true
    bOverrideAngularDamping=true
    bUseGravityOverride=true
    linearSpringConstant=10000
    boundBone {boneName=Shoulder_r_jnt}
    componentAppliedLinearAccClamp=<3000 3000 3000>
    componentLinearAccScale=<0.1 0.1 0.1>
    componentLinearVelScale=<0.1 0.1 0.1>
    gravityOverride {z=-980}
    physicsBodyDefinitions [{
      boundBone {boneName=Shoulder_r_jnt}
      boxExtents=<10 10 10>
      constraintSetup {angularLimitsMax=<10 10 10>, angularLimitsMin=<-10 -10 -10>, angularTarget=<1 0 0>}
      localJointOffset=<-1 0 0>
     }]
    relativeSpaceBone {boneName=Root_Clock_jnt}
   }
   n8 = Anim Dynamics - Bone: Shoulder_l_jnt(ComponentPose=@n7, Alpha=1, ExternalForce="0,0,0") {
    angularDampingOverride=0.999
    angularSpringConstant=20000
    bAngularSpring=true
    bOverrideAngularDamping=true
    bUseGravityOverride=true
    linearSpringConstant=10000
    boundBone {boneName=Shoulder_l_jnt}
    componentAppliedLinearAccClamp=<3000 3000 3000>
    componentLinearAccScale=<0.1 0.1 0.1>
    componentLinearVelScale=<0.1 0.1 0.1>
    gravityOverride {z=-600}
    physicsBodyDefinitions [{
      boundBone {boneName=Shoulder_l_jnt}
      boxExtents=<10 10 10>
      constraintSetup {angularLimitsMax=<10 10 10>, angularLimitsMin=<-10 -10 -10>, angularTarget=<1 0 0>}
      localJointOffset=<-1 0 0>
     }]
    relativeSpaceBone {boneName=Root_Clock_jnt}
   }
   n9 = Anim Dynamics - Bone: Foot_l_jnt(ComponentPose=@n8, Alpha=1, ExternalForce="0,0,0") {
    angularDampingOverride=1
    angularSpringConstant=20000
    bAngularSpring=true
    bGravityOverrideInSimSpace=true
    bOverrideAngularDamping=true
    bUseGravityOverride=true
    linearSpringConstant=10000
    boundBone {boneName=Foot_l_jnt}
    componentAppliedLinearAccClamp=<3000 3000 3000>
    componentLinearAccScale=<0.5 0.5 0.5>
    componentLinearVelScale=<0.5 0.5 0.5>
    gravityOverride {z=-250}
    physicsBodyDefinitions [{
      boundBone {boneName=Foot_l_jnt}
      boxExtents=<10 10 10>
      constraintSetup {angularLimitsMax=<5 5 5>, angularLimitsMin=<-5 -5 -5>, angularTarget=<1 0 0>}
      localJointOffset=<-1 0 0>
     }]
    relativeSpaceBone {boneName=Root_Clock_jnt}
   }
   n10 = Anim Dynamics - Bone: Foot_r_jnt(ComponentPose=@n9, Alpha=1, ExternalForce="0,0,0") {
    angularDampingOverride=0.9999
    angularSpringConstant=20000
    bAngularSpring=true
    bGravityOverrideInSimSpace=true
    bOverrideAngularDamping=true
    bUseGravityOverride=true
    linearSpringConstant=10000
    boundBone {boneName=Foot_r_jnt}
    componentAppliedLinearAccClamp=<3000 3000 3000>
    componentLinearAccScale=<0.5 0.5 0.5>
    componentLinearVelScale=<0.5 0.5 0.5>
    gravityOverride {z=-250}
    physicsBodyDefinitions [{
      boundBone {boneName=Foot_r_jnt}
      boxExtents=<10 10 10>
      constraintSetup {angularLimitsMax=<5 5 5>, angularLimitsMin=<-5 -5 -5>, angularTarget=<1 0 0>}
      localJointOffset=<-1 0 0>
     }]
    relativeSpaceBone {boneName=Root_Clock_jnt}
   }
   n11 = Anim Dynamics - Bone: Head_jnt(ComponentPose=Local To Component(LocalPose=Use cached pose 'SavedInputPose'()), Alpha=1, ExternalForce="0,0,0") {
    angularDampingOverride=1
    angularSpringConstant=5000
    bAngularSpring=true
    bGravityOverrideInSimSpace=true
    bOverrideAngularDamping=true
    bUseGravityOverride=true
    linearSpringConstant=10000
    boundBone {boneName=Head_jnt}
    componentAppliedLinearAccClamp=<10000 10000 10000>
    componentLinearAccScale=<0.1 0.1 0.1>
    componentLinearVelScale=<0.1 0.1 0.1>
    gravityOverride {z=-2}
    physicsBodyDefinitions [{
      boundBone {boneName=Head_jnt}
      boxExtents=<10 10 10>
      constraintSetup {angularLimitsMax=<2 2 2>, angularLimitsMin=<-2 -2 -2>, angularTarget=<1 0 0>}
      localJointOffset=<-2 0 0>
     }]
    relativeSpaceBone {boneName=Root_Clock_jnt}
    simSpaceSettings {maxAngularAcceleration=35, maxAngularVelocity=7, simSpaceAngularAlpha=1}
   }
   n12 = Anim Dynamics - Bone: Hair_01_l_jnt(ComponentPose=@n11, Alpha=1, ExternalForce="0,0,0") {
    angularDampingOverride=1
    angularSpringConstant=5000
    bAngularSpring=true
    bGravityOverrideInSimSpace=true
    bOverrideAngularDamping=true
    bUseGravityOverride=true
    linearSpringConstant=10000
    boundBone {boneName=Hair_01_l_jnt}
    componentAppliedLinearAccClamp=<10000 10000 10000>
    componentLinearAccScale=<0.1 0.1 0.1>
    componentLinearVelScale=<0.1 0.1 0.1>
    gravityOverride {z=-2}
    physicsBodyDefinitions [{
      boundBone {boneName=Hair_01_l_jnt}
      boxExtents=<10 10 10>
      constraintSetup {angularLimitsMax=<2 2 2>, angularLimitsMin=<-2 -2 -2>, angularTarget=<1 0 0>}
      localJointOffset=<-2 0 0>
     }]
    relativeSpaceBone {boneName=Root_Clock_jnt}
    simSpaceSettings {maxAngularAcceleration=35, maxAngularVelocity=7, simSpaceAngularAlpha=1}
   }
   n13 = Anim Dynamics - Bone: Hair_01_r_jnt(ComponentPose=@n12, Alpha=1, ExternalForce="0,0,0") {
    angularDampingOverride=1
    angularSpringConstant=5000
    bAngularSpring=true
    bGravityOverrideInSimSpace=true
    bOverrideAngularDamping=true
    bUseGravityOverride=true
    linearSpringConstant=10000
    boundBone {boneName=Hair_01_r_jnt}
    componentAppliedLinearAccClamp=<10000 10000 10000>
    componentLinearAccScale=<0.1 0.1 0.1>
    componentLinearVelScale=<0.1 0.1 0.1>
    gravityOverride {z=-2}
    physicsBodyDefinitions [{
      boundBone {boneName=Hair_01_r_jnt}
      boxExtents=<10 10 10>
      constraintSetup {angularLimitsMax=<2 2 2>, angularLimitsMin=<-2 -2 -2>, angularTarget=<1 0 0>}
      localJointOffset=<-2 0 0>
     }]
    relativeSpaceBone {boneName=Root_Clock_jnt}
    simSpaceSettings {maxAngularAcceleration=35, maxAngularVelocity=7, simSpaceAngularAlpha=1}
   }
   n14 = Anim Dynamics - Bone: Shoulder_l_jnt(ComponentPose=@n13, Alpha=1, ExternalForce="0,0,0") {
    angularDampingOverride=1
    angularSpringConstant=5000
    bAngularSpring=true
    bGravityOverrideInSimSpace=true
    bOverrideAngularDamping=true
    bUseGravityOverride=true
    linearSpringConstant=10000
    boundBone {boneName=Shoulder_l_jnt}
    componentAppliedLinearAccClamp=<10000 10000 10000>
    componentLinearAccScale=<0.1 0.1 0.1>
    componentLinearVelScale=<0.1 0.1 0.1>
    gravityOverride {z=-2}
    physicsBodyDefinitions [{
      boundBone {boneName=Shoulder_l_jnt}
      boxExtents=<10 10 10>
      constraintSetup {angularLimitsMax=<2 2 2>, angularLimitsMin=<-2 -2 -2>, angularTarget=<1 0 0>}
      localJointOffset=<-2 0 0>
     }]
    relativeSpaceBone {boneName=Root_Clock_jnt}
    simSpaceSettings {maxAngularAcceleration=35, maxAngularVelocity=7, simSpaceAngularAlpha=1}
   }
   n15 = Anim Dynamics - Bone: Shoulder_r_jnt(ComponentPose=@n14, Alpha=1, ExternalForce="0,0,0") {
    angularDampingOverride=1
    angularSpringConstant=5000
    bAngularSpring=true
    bGravityOverrideInSimSpace=true
    bOverrideAngularDamping=true
    bUseGravityOverride=true
    linearSpringConstant=10000
    boundBone {boneName=Shoulder_r_jnt}
    componentAppliedLinearAccClamp=<10000 10000 10000>
    componentLinearAccScale=<0.1 0.1 0.1>
    componentLinearVelScale=<0.1 0.1 0.1>
    gravityOverride {z=-2}
    physicsBodyDefinitions [{
      boundBone {boneName=Shoulder_r_jnt}
      boxExtents=<10 10 10>
      constraintSetup {angularLimitsMax=<2 2 2>, angularLimitsMin=<-2 -2 -2>, angularTarget=<1 0 0>}
      localJointOffset=<-2 0 0>
     }]
    relativeSpaceBone {boneName=Root_Clock_jnt}
    simSpaceSettings {maxAngularAcceleration=35, maxAngularVelocity=7, simSpaceAngularAlpha=1}
   }
   n16 = Anim Dynamics - Bone: Foot_l_jnt(ComponentPose=@n15, Alpha=1, ExternalForce="0,0,0") {
    angularDampingOverride=1
    angularSpringConstant=5000
    bAngularSpring=true
    bGravityOverrideInSimSpace=true
    bOverrideAngularDamping=true
    bUseGravityOverride=true
    linearSpringConstant=10000
    boundBone {boneName=Foot_l_jnt}
    componentAppliedLinearAccClamp=<10000 10000 10000>
    componentLinearAccScale=<0.1 0.1 0.1>
    componentLinearVelScale=<0.1 0.1 0.1>
    gravityOverride {z=-2}
    physicsBodyDefinitions [{
      boundBone {boneName=Foot_l_jnt}
      boxExtents=<10 10 10>
      constraintSetup {angularLimitsMax=<2 2 2>, angularLimitsMin=<-2 -2 -2>, angularTarget=<1 0 0>}
      localJointOffset=<-2 0 0>
     }]
    relativeSpaceBone {boneName=Root_Clock_jnt}
    simSpaceSettings {maxAngularAcceleration=35, maxAngularVelocity=7, simSpaceAngularAlpha=1}
   }
   Save cached pose 'AnimDynamic'(Pose=Two Way Blend(A=Component To Local(ComponentPose=@n10), B=Component To Local(ComponentPose=@n16), bAlphaBoolEnabled=Get bIsInFrontEnd()) {blendNode {alphaInputType=Bool}}) {cacheName=AnimDynamic}
  ]}
 EventGraph {nodes [
   Event BlueprintUpdateAnimation()
   > n1 = Set GameSeconds(GameSeconds:double=float * float(A=GetGameTimeInSeconds(), B=Get TimeScale() -> (TimeScale:double)) {is=title K2Node_PromotableOperator}) -> (Output_Get:double)
   > n2 = Sequence()
   > n3 = Set MinutesFloat(MinutesFloat:double="% (Float)"!(A:double=float / float(A=@n1, B=60), B:double=60) -> (ReturnValue:double)) -> (Output_Get:double)
   n4 = float * float(A=@n3, B=6)
   > Set MinuteAngle(MinuteAngle:double=float + float(A=float * float(A=@n4, B=-1), B=36)) -> (Output_Get:double)
   n5 = Set HoursFloat(execute=@n2.then_1, HoursFloat:double="% (Float)"(A=float / float(A=@n1, B=3600), B=12)) -> (Output_Get:double)
   n6 = float * float(A=@n5, B=30)
   > Set HourAngle(HourAngle:double=float + float(A=float * float(A=@n6, B=-1), B=-157)) -> (Output_Get:double)
   n7 = "To Float (Integer)"(InInt=Truncate(A=@n5))
   n8 = "To Float (Integer)"(InInt=Truncate(A=@n3))
   PrintString!(
    execute
    InString="To String (Text)"(InText=Format Text!(Format="LOCTEXT(\"Min {min}\\r\\nHour {hour}\\r\\n\")", min:double=@n8, hour:double=@n7) -> (Result))
    bPrintToScreen
    bPrintToLog
    TextColor
    Duration
    Key=jhgj
   ) -> (then)
  ]}
]
variables [
 GameSeconds {type=number}
 HourAngle {type=number, default=90}
 HoursFloat {type=number}
 MinuteAngle {type=number, default=90}
 MinutesFloat {type=number}
 TimeScale {type=number, default=100}
]
