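# Niagara system for the QueenTruth Rind pickaxe's toon-style trail, built from two short-lived ribbon emitters.
# Round_Trail and TopFacingTrail spawn through anim-trail functions, each tracking different bone or socket indices.
# Both fade alpha by curve, dim RGB at night, and set outline width and color alongside two ribbon renderers for the toon look.
# Round_Trail also narrows over its loop, while TopFacingTrail fades its outline alpha.
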NiagaraSystem /BRCosmetics/Weapons/FORT_Melee/Pickaxe_QueenTruth_Rind/VFX/NS_Pickaxe_QueenTruth_Rind_ToonTrail
dependencies {
 usedDataInterfaces [&NiagaraDataInterfaceCurve, &NiagaraDataInterfaceSocketReader]
 usedDynamicInputs []
 usedModules [
  &self:SystemSpawnScript.SystemScriptSource.SystemScriptGraph.NiagaraNodeAssignment_0.SetVariables_#1
  &SystemState
 ]
}
emitters {
 Round_Trail {
  emitterUpdateScript [
   &EmitterState(Life Cycle Mode=::NewEnumerator1, Loop Behavior=::NewEnumerator1, Loop Duration=0.25, UseLoopDelay=false)
   &FN_SpawnAnimTrail()
  ]
  particleSpawnScript [
   &FN_InitializeAnimTrailBoneIndecies(First Bone or Socket Index=1, Second Bone or Socket Index=0)
   &FN_InitializeAnimTrail(Color=<rgba 1 1 1 0.5>, Lifetime=0.2)
  ]
  particleUpdateScript [
   &FN_UpdateAnimTrail() {module=FN_UpdateAnimTrail001}
   &SolveForcesAndVelocity(Clamp Velocity=false, Limit Acceleration=false)
   &ScaleColor(
    Scale Alpha=&FloatFromCurve(FloatCurve=DataInterface() {propertyValues {Curve {keys [
         {arriveTangent=-11.5755, interpMode=RCIM_Cubic, leaveTangent=-11.5755, tangentMode=RCTM_User, time=0, value=1.25}
         {arriveTangent=0.1134, interpMode=RCIM_Cubic, leaveTangent=0.1134, tangentMode=RCTM_User, time=0.2, value=0}
        ]}}})
    Scale RGB=&Multiply_Vector(
     A=<1 1 1>
     B=&FN_Select_Vector_from_Time_Of_Day(
      Morning=&MakeVectorFromLinearColorRGB(Color=<rgba 1 1 1 1>)
      Day=&MakeVectorFromLinearColorRGB(Color=<rgba 1 1 1 1>)
      Evening=&MakeVectorFromLinearColorRGB(Color=<rgba 1 1 1 1>)
      Night=&MakeVectorFromLinearColorRGB(Color=<rgba 0.33 0.33 0.33 1>)
     )
    )
    ScaleA=true
    ScaleRGB=true
   ) {module=ScaleColor001}
   /SetVariables(Particles.OutlineWidth:NiagaraFloat=&Multiply_Float(A=$Particles.RibbonWidth, B=0.95)) {isSetParameters=true}
   /SetVariables(Particles.OutlineColor:LinearColor=<rgba 0 0 0 0.9>) {isSetParameters=true}
   &ScaleRibbonWidth(Ribbon Width Scale=&FloatFromCurve(
     FloatCurve=DataInterface() {propertyValues {Curve {keys [{arriveTangent=-1.7104, interpMode=RCIM_Cubic, leaveTangent=-1.7104, tangentMode=RCTM_User, time=0.4, value=1}, <at 1 0.5 RCIM_Cubic>]}}}
     CurveIndex=$Emitter.NormalizedLoopAge
    ))
  ]
  rendererClasses [&NiagaraRibbonRendererProperties]
  renderers [
   {rendererClass=&NiagaraRibbonRendererProperties}
   {rendererClass=&NiagaraRibbonRendererProperties, rendererIndex=1}
  ]
 }
 TopFacingTrail {
  emitterUpdateScript [
   &EmitterState(Life Cycle Mode=::NewEnumerator1, Loop Behavior=::NewEnumerator1, Loop Duration=0.25, UseLoopDelay=false)
   &FN_SpawnAnimTrail()
  ]
  particleSpawnScript [
   &FN_InitializeAnimTrailBoneIndecies(First Bone or Socket Index=2, Second Bone or Socket Index=2)
   &FN_InitializeAnimTrail(AnimTrailWidth=15, Color=<rgba 1 1 1 0.5>, Lifetime=0.225)
  ]
  particleUpdateScript [
   &FN_UpdateAnimTrail() {module=FN_UpdateAnimTrail001}
   &SolveForcesAndVelocity(Clamp Velocity=false, Limit Acceleration=false)
   &ScaleColor(
    Scale Alpha=&FloatFromCurve(FloatCurve=DataInterface() {propertyValues {Curve {keys [
         {arriveTangent=0.0494, interpMode=RCIM_Cubic, leaveTangent=0.0494, tangentMode=RCTM_User, time=0, value=1.25}
         {interpMode=RCIM_Cubic, tangentMode=RCTM_User, time=0.2, value=0}
        ]}}})
    Scale RGB=&Multiply_Vector(
     A=<1 1 1>
     B=&FN_Select_Vector_from_Time_Of_Day(
      Morning=&MakeVectorFromLinearColorRGB(Color=<rgba 1 1 1 1>)
      Day=&MakeVectorFromLinearColorRGB(Color=<rgba 1 1 1 1>)
      Evening=&MakeVectorFromLinearColorRGB(Color=<rgba 1 1 1 1>)
      Night=&MakeVectorFromLinearColorRGB(Color=<rgba 0.33 0.33 0.33 1>)
     )
    )
    ScaleA=true
    ScaleRGB=true
   ) {module=ScaleColor001}
   /SetVariables(Particles.OutlineWidth:NiagaraFloat=&Multiply_Float(A=$Particles.RibbonWidth, B=1.1)) {isSetParameters=true}
   /SetVariables(Particles.OutlineColor:LinearColor=&MultiplyLinearColor(
     Linear Color=<rgba 0 0 0 0.9>
     Scale Factor=&MakeVector4FromVectorAndFloat(
      "Vector (RGB)"=<1 1 1>
      "Float (Alpha)"=&FloatFromCurve(FloatCurve=DataInterface() {propertyValues {Curve {keys [
           {arriveTangent=2.0405, interpMode=RCIM_Cubic, leaveTangent=2.0405, tangentMode=RCTM_User, time=0, value=1.25}
           {interpMode=RCIM_Cubic, tangentMode=RCTM_User, time=0.25, value=0}
          ]}}})
     )
    )) {isSetParameters=true}
  ]
  rendererClasses [&NiagaraRibbonRendererProperties]
  renderers [
   {rendererClass=&NiagaraRibbonRendererProperties}
   {rendererClass=&NiagaraRibbonRendererProperties, rendererIndex=1}
  ]
 }
}
summary {
 systemName=NS_Pickaxe_QueenTruth_Rind_ToonTrail
 emitters [
  {emitterName=Round_Trail, rendererClasses [&NiagaraRibbonRendererProperties]}
  {emitterName=TopFacingTrail, rendererClasses [&NiagaraRibbonRendererProperties]}
 ]
 userVariables [User.RibbonIDOffset {type=NiagaraInt32}]
}
